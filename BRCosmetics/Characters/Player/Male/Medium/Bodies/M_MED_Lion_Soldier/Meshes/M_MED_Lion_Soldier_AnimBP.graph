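# Animation Blueprint for the Lion Soldier cosmetic body, adding procedural secondary motion to a pose copied from its attached parent mesh.
# Weighted constraints drive shoulder shells, pouches and kneepads, with AnimDynamics jiggling pouches; shell dynamics are scaled to zero.
# Update-time bone-rotation deltas and crouch events set alphas for additive corrective bones at shoulder, knees, left hip and left pouch.
# An earlier shell constraint chain is disconnected, and `arm_raise_r` is unused.
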AnimBlueprint /BRCosmetics/Characters/Player/Male/Medium/Bodies/M_MED_Lion_Soldier/Meshes/M_MED_Lion_Soldier_AnimBP
target_skeleton=&/M_MED_Lion_Soldier_Skeleton
graphs [
 AnimGraph {nodes [
   n1 = "Constraint - shell_1 (PPPPPP)"!(
    ComponentPose
    Alpha=1
    ConstraintWeights_0=0.0682
    ConstraintWeights_1:float=0.0847
    ConstraintWeights_2:float=0.0301
    ConstraintWeights_3:float=0.2693
    ConstraintWeights_4:float=0.5223
    ConstraintWeights_5:float=0.0253
   ) -> (Pose) {
    alphaBoolBlend {blendOption=Linear}
    boneToModify {boneName=shell_1}
    constraintSetup [
     {targetBone {boneName=upperarm_twist_01_r}}
     {targetBone {boneName=spine_03}}
     {targetBone {boneName=spine_04}}
     {targetBone {boneName=spine_05}}
     {targetBone {boneName=clavicle_r}}
     {targetBone {boneName=neck_01}}
    ]
    constraintWeights [1,1,1,1,1,1]
   }
   n2 = "Constraint - shell_2 (PPPPPPPP)"!(
    ComponentPose=@n1
    Alpha=1
    ConstraintWeights_0=0
    ConstraintWeights_1:float=0.0767
    ConstraintWeights_2:float=0
    ConstraintWeights_3:float=0.0527
    ConstraintWeights_4:float=0.0095
    ConstraintWeights_5:float=0.2582
    ConstraintWeights_6:float=0.5444
    ConstraintWeights_7:float=0.0586
   ) -> (Pose) {
    alphaBoolBlend {blendOption=Linear}
    boneToModify {boneName=shell_2}
    constraintSetup [
     {targetBone {boneName=upperarm_r}}
     {targetBone {boneName=upperarm_twist_01_r}}
     {targetBone {boneName=upperarm_twist_02_r}}
     {targetBone {boneName=spine_03}}
     {targetBone {boneName=spine_04}}
     {targetBone {boneName=spine_05}}
     {targetBone {boneName=clavicle_r}}
     {targetBone {boneName=neck_01}}
    ]
    constraintWeights [1,1,1,1,1,1,1,1]
   }
   n3 = "Constraint - shell_3 (PPPPPPPP)"!(
    ComponentPose=@n2
    Alpha=1
    ConstraintWeights_0=0
    ConstraintWeights_1:float=0.0739
    ConstraintWeights_2:float=0
    ConstraintWeights_3:float=0.0398
    ConstraintWeights_4:float=0.0123
    ConstraintWeights_5:float=0.2681
    ConstraintWeights_6:float=0.5381
    ConstraintWeights_7:float=0.0678
   ) -> (Pose) {
    alphaBoolBlend {blendOption=Linear}
    boneToModify {boneName=shell_3}
    constraintSetup [
     {targetBone {boneName=upperarm_r}}
     {targetBone {boneName=upperarm_twist_01_r}}
     {targetBone {boneName=upperarm_twist_02_r}}
     {targetBone {boneName=spine_03}}
     {targetBone {boneName=spine_04}}
     {targetBone {boneName=spine_05}}
     {targetBone {boneName=clavicle_r}}
     {targetBone {boneName=neck_01}}
    ]
    constraintWeights [1,1,1,1,1,1,1,1]
   }
   n4 = "Constraint - shell_4 (PPPPPPPP)"!(
    ComponentPose=@n3
    Alpha=1
    ConstraintWeights_0=0
    ConstraintWeights_1:float=0.0752
    ConstraintWeights_2:float=0
    ConstraintWeights_3:float=0.0331
    ConstraintWeights_4:float=0.029
    ConstraintWeights_5:float=0.2709
    ConstraintWeights_6:float=0.5388
    ConstraintWeights_7:float=0.0529
   ) -> (Pose) {
    alphaBoolBlend {blendOption=Linear}
    boneToModify {boneName=shell_4}
    constraintSetup [
     {targetBone {boneName=upperarm_r}}
     {targetBone {boneName=upperarm_twist_01_r}}
     {targetBone {boneName=upperarm_twist_02_r}}
     {targetBone {boneName=spine_03}}
     {targetBone {boneName=spine_04}}
     {targetBone {boneName=spine_05}}
     {targetBone {boneName=clavicle_r}}
     {targetBone {boneName=neck_01}}
    ]
    constraintWeights [1,1,1,1,1,1,1,1]
   }
   "Constraint - shell_5 (PPPPPPPP)"!(
    ComponentPose=@n4
    Alpha=1
    ConstraintWeights_0=0
    ConstraintWeights_1:float=0.036
    ConstraintWeights_2:float=0
    ConstraintWeights_3:float=0.058
    ConstraintWeights_4:float=0.0561
    ConstraintWeights_5:float=0.2593
    ConstraintWeights_6:float=0.5827
    ConstraintWeights_7:float=0.0079
   ) -> (Pose) {
    alphaBoolBlend {blendOption=Linear}
    boneToModify {boneName=shell_5}
    constraintSetup [
     {targetBone {boneName=upperarm_r}}
     {targetBone {boneName=upperarm_twist_01_r}}
     {targetBone {boneName=upperarm_twist_02_r}}
     {targetBone {boneName=spine_03}}
     {targetBone {boneName=spine_04}}
     {targetBone {boneName=spine_05}}
     {targetBone {boneName=clavicle_r}}
     {targetBone {boneName=neck_01}}
    ]
    constraintWeights [1,1,1,1,1,1,1,1]
   }
   n5 = Get shldr_raise_r() -> (shldr_raise_r:double)
   n6 = "Transform (Modify) Bone - Bone: clav_slide_r"(
    ComponentPose=Local To Component(LocalPose=Copy Pose From Mesh() {bUseAttachedParent=true})
    Alpha=@n5
    Translation="8,0,-3"
    Rotation="0,0,0"
    Scale="1,1,1"
   ) {
    rotationMode=BMM_Additive
    rotationSpace=BCS_BoneSpace
    translationMode=BMM_Additive
    translationSpace=BCS_BoneSpace
    alphaBoolBlend {blendOption=Linear}
    alphaScaleBiasClamp {scale=0.8}
    boneToModify {boneName=clav_slide_r}
    translation {x=8, z=-3}
   }
   n7 = "Transform (Modify) Bone - Bone: collar_fr_r"(ComponentPose=@n6, Alpha=@n5, Translation="0,1,0", Rotation="0,0,0", Scale="1,1,1") {
    rotationMode=BMM_Additive
    rotationSpace=BCS_BoneSpace
    translationMode=BMM_Additive
    translationSpace=BCS_BoneSpace
    alphaBoolBlend {blendOption=Linear}
    boneToModify {boneName=collar_fr_r}
    translation {y=1}
   }
   n8 = "Transform (Modify) Bone - Bone: collar_r"(ComponentPose=@n7, Alpha=@n5, Translation="0,1,0", Rotation="0,0,0", Scale="1,1,1") {
    rotationMode=BMM_Additive
    rotationSpace=BCS_BoneSpace
    translationMode=BMM_Additive
    translationSpace=BCS_BoneSpace
    alphaBoolBlend {blendOption=Linear}
    boneToModify {boneName=collar_r}
    translation {y=1}
   }
   n9 = "Constraint - shell_1 (PPPPPPP)"!(
    ComponentPose=@n8
    Alpha=1
    ConstraintWeights_0=0.0509
    ConstraintWeights_1:float=0.0198
    ConstraintWeights_2:float=0.0901
    ConstraintWeights_3:float=0.0395
    ConstraintWeights_4:float=0.2926
    ConstraintWeights_5:float=0.4821
    ConstraintWeights_6:float=0.025
   ) -> (Pose) {
    alphaBoolBlend {blendOption=Linear}
    boneToModify {boneName=shell_1}
    constraintSetup [
     {targetBone {boneName=upperarm_twist_01_r}}
     {targetBone {boneName=clav_slide_r}}
     {targetBone {boneName=spine_03}}
     {targetBone {boneName=spine_04}}
     {targetBone {boneName=spine_05}}
     {targetBone {boneName=clavicle_r}}
     {targetBone {boneName=neck_01}}
    ]
    constraintWeights [1,1,1,1,1,1,1]
   }
   n10 = "Constraint - shell_2 (PPPPPPP)"!(
    ComponentPose=@n9
    Alpha=1
    ConstraintWeights_0=0.0775
    ConstraintWeights_1:float=0.2151
    ConstraintWeights_2:float=0.0526
    ConstraintWeights_3:float=0.0092
    ConstraintWeights_4:float=0.2591
    ConstraintWeights_5:float=0.3274
    ConstraintWeights_6:float=0.0592
   ) -> (Pose) {
    alphaBoolBlend {blendOption=Linear}
    boneToModify {boneName=shell_2}
    constraintSetup [
     {targetBone {boneName=upperarm_twist_01_r}}
     {targetBone {boneName=clav_slide_r}}
     {targetBone {boneName=spine_03}}
     {targetBone {boneName=spine_04}}
     {targetBone {boneName=spine_05}}
     {targetBone {boneName=clavicle_r}}
     {targetBone {boneName=neck_01}}
    ]
    constraintWeights [1,1,1,1,1,1,1]
   }
   n11 = "Constraint - shell_3 (PPPPPPP)"!(
    ComponentPose=@n10
    Alpha=1
    ConstraintWeights_0=0.0748
    ConstraintWeights_1:float=0.3672
    ConstraintWeights_2:float=0.0399
    ConstraintWeights_3:float=0.0123
    ConstraintWeights_4:float=0.2682
    ConstraintWeights_5:float=0.1699
    ConstraintWeights_6:float=0.0678
   ) -> (Pose) {
    alphaBoolBlend {blendOption=Linear}
    boneToModify {boneName=shell_3}
    constraintSetup [
     {targetBone {boneName=upperarm_twist_01_r}}
     {targetBone {boneName=clav_slide_r}}
     {targetBone {boneName=spine_03}}
     {targetBone {boneName=spine_04}}
     {targetBone {boneName=spine_05}}
     {targetBone {boneName=clavicle_r}}
     {targetBone {boneName=neck_01}}
    ]
    constraintWeights [1,1,1,1,1,1,1]
   }
   n12 = "Constraint - shell_4 (PPPPPPPP)"(
    ComponentPose=@n11
    Alpha=1
    ConstraintWeights_0=0.084
    ConstraintWeights_1=0.4463
    ConstraintWeights_2=0.0001
    ConstraintWeights_3=0.0328
    ConstraintWeights_4=0.0291
    ConstraintWeights_5=0.2853
    ConstraintWeights_6=0.0619
    ConstraintWeights_7=0.0604
   ) {
    alphaBoolBlend {blendOption=Linear}
    boneToModify {boneName=shell_4}
    constraintSetup [
     {targetBone {boneName=upperarm_twist_01_r}}
     {targetBone {boneName=clav_slide_r}}
     {targetBone {boneName=collar_r}}
     {targetBone {boneName=spine_03}}
     {targetBone {boneName=spine_04}}
     {targetBone {boneName=spine_05}}
     {targetBone {boneName=clavicle_r}}
     {targetBone {boneName=neck_01}}
    ]
    constraintWeights [1,1,1,1,1,1,1,1]
   }
   n13 = "Constraint - shell_5 (PPPPPPPP)"(
    ComponentPose=@n12
    Alpha=1
    ConstraintWeights_0=0.0745
    ConstraintWeights_1=0.3715
    ConstraintWeights_2=0.0001
    ConstraintWeights_3=0.0487
    ConstraintWeights_4=0.0698
    ConstraintWeights_5=0.353
    ConstraintWeights_6=0.0348
    ConstraintWeights_7=0.0476
   ) {
    alphaBoolBlend {blendOption=Linear}
    boneToModify {boneName=shell_5}
    constraintSetup [
     {targetBone {boneName=upperarm_twist_01_r}}
     {targetBone {boneName=clav_slide_r}}
     {targetBone {boneName=collar_r}}
     {targetBone {boneName=spine_03}}
     {targetBone {boneName=spine_04}}
     {targetBone {boneName=spine_05}}
     {targetBone {boneName=clavicle_r}}
     {targetBone {boneName=neck_01}}
    ]
    constraintWeights [1,1,1,1,1,1,1,1]
   }
   n14 = Anim Dynamics - Bone: shell_1(ComponentPose=@n13, Alpha=1, ExternalForce="0,0,0") {
    angularDampingOverride=1
    angularSpringConstant=1500
    bAngularSpring=true
    bGravityOverrideInSimSpace=true
    bOverrideAngularDamping=true
    gravityScale=0.35
    simulationSpace=RootRelative
    alphaBoolBlend {blendOption=Linear}
    alphaScaleBiasClamp {scale=0}
    boundBone {boneName=shell_1}
    physicsBodyDefinitions [{
      sphereCollisionRadius=0
      boundBone {boneName=shell_1}
      boxExtents=<10 20 20>
      constraintSetup {angularLimitsMax=<0 40 -40>, angularLimitsMin=<0 -40 -40>, angularTarget=<0 0 0>}
      localJointOffset=<15 0 0>
     }]
   }
   n15 = Anim Dynamics - Bone: shell_2(ComponentPose=@n14, Alpha=1, ExternalForce="0,0,0") {
    angularDampingOverride=1
    angularSpringConstant=1500
    bAngularSpring=true
    bGravityOverrideInSimSpace=true
    bOverrideAngularDamping=true
    gravityScale=0.35
    simulationSpace=RootRelative
    alphaBoolBlend {blendOption=Linear}
    alphaScaleBiasClamp {scale=0}
    boundBone {boneName=shell_2}
    physicsBodyDefinitions [{
      sphereCollisionRadius=0
      boundBone {boneName=shell_2}
      boxExtents=<10 20 20>
      constraintSetup {angularLimitsMax=<0 40 -40>, angularLimitsMin=<0 -40 -40>, angularTarget=<0 0 0>}
      localJointOffset=<12 0 0>
     }]
   }
   n16 = Anim Dynamics - Bone: shell_3(ComponentPose=@n15, Alpha=1, ExternalForce="0,0,0") {
    angularDampingOverride=1
    angularSpringConstant=1500
    bAngularSpring=true
    bGravityOverrideInSimSpace=true
    bOverrideAngularDamping=true
    gravityScale=0.35
    simulationSpace=RootRelative
    alphaBoolBlend {blendOption=Linear}
    alphaScaleBiasClamp {scale=0}
    boundBone {boneName=shell_3}
    physicsBodyDefinitions [{
      sphereCollisionRadius=0
      boundBone {boneName=shell_3}
      boxExtents=<10 20 20>
      constraintSetup {angularLimitsMax=<0 40 -40>, angularLimitsMin=<0 -40 -40>, angularTarget=<0 0 0>}
      localJointOffset=<10 0 0>
     }]
   }
   n17 = Anim Dynamics - Bone: shell_4(ComponentPose=@n16, Alpha=1, ExternalForce="0,0,0") {
    angularDampingOverride=1
    angularSpringConstant=1500
    bAngularSpring=true
    bGravityOverrideInSimSpace=true
    bOverrideAngularDamping=true
    gravityScale=0.35
    simulationSpace=RootRelative
    alphaBoolBlend {blendOption=Linear}
    alphaScaleBiasClamp {scale=0}
    boundBone {boneName=shell_4}
    physicsBodyDefinitions [{
      sphereCollisionRadius=0
      boundBone {boneName=shell_4}
      boxExtents=<10 20 20>
      constraintSetup {angularLimitsMax=<0 40 -40>, angularLimitsMin=<0 -40 -40>, angularTarget=<0 0 0>}
      localJointOffset=<13 0 0>
     }]
   }
   n18 = Anim Dynamics - Bone: shell_5(ComponentPose=@n17, Alpha=1, ExternalForce="0,0,0") {
    angularDampingOverride=1
    angularSpringConstant=1500
    bAngularSpring=true
    bGravityOverrideInSimSpace=true
    bOverrideAngularDamping=true
    gravityScale=0.35
    simulationSpace=RootRelative
    alphaBoolBlend {blendOption=Linear}
    alphaScaleBiasClamp {scale=0}
    boundBone {boneName=shell_5}
    physicsBodyDefinitions [{
      sphereCollisionRadius=0
      boundBone {boneName=shell_5}
      boxExtents=<10 20 20>
      constraintSetup {angularLimitsMax=<0 40 -40>, angularLimitsMin=<0 -40 -40>, angularTarget=<0 0 0>}
      localJointOffset=<16 0 0>
     }]
   }
   n19 = "Constraint - pouch_1_r (PPPPP)"!(
    ComponentPose=@n18
    Alpha=1
    ConstraintWeights_0=0.1852
    ConstraintWeights_1:float=0.0071
    ConstraintWeights_2:float=0.0075
    ConstraintWeights_3:float=0.0064
    ConstraintWeights_4:float=0.7938
   ) -> (Pose) {
    alphaBoolBlend {blendOption=Linear}
    boneToModify {boneName=pouch_1_r}
    constraintSetup [
     {targetBone {boneName=spine_01}}
     {targetBone {boneName=spine_02}}
     {targetBone {boneName=spine_03}}
     {targetBone {boneName=thigh_twist_01_r}}
     {targetBone {boneName=pelvis}}
    ]
    constraintWeights [1,1,1,1,1]
   }
   n20 = "Constraint - pouch_1_l (PPPPP)"!(
    ComponentPose=@n19
    Alpha=1
    ConstraintWeights_0=0.0037
    ConstraintWeights_1:float=0.2489
    ConstraintWeights_2:float=0.0184
    ConstraintWeights_3:float=0.0196
    ConstraintWeights_4:float=0.7094
   ) -> (Pose) {
    alphaBoolBlend {blendOption=Linear}
    boneToModify {boneName=pouch_1_l}
    constraintSetup [
     {targetBone {boneName=spine_03}}
     {targetBone {boneName=spine_01}}
     {targetBone {boneName=spine_02}}
     {targetBone {boneName=thigh_twist_01_l}}
     {targetBone {boneName=pelvis}}
    ]
    constraintWeights [1,1,1,1,1]
   }
   n21 = "Transform (Modify) Bone - Bone: pouch_1_l"(ComponentPose=@n20, Alpha=Get crouch_alpha() -> (crouch_alpha:double), Translation="0,-2,1", Rotation="0,0,-20", Scale="1,1,1") {
    rotationMode=BMM_Additive
    rotationSpace=BCS_BoneSpace
    translationMode=BMM_Additive
    translationSpace=BCS_ParentBoneSpace
    alphaBoolBlend {blendOption=Linear}
    alphaScaleBiasClamp {bInterpResult=true}
    boneToModify {boneName=pouch_1_l}
    rotation {roll=-20}
    translation {y=-2, z=1}
   }
   n22 = Anim Dynamics - Bone: pouch_2_r(ComponentPose=@n21, Alpha=1, ExternalForce="0,0,0") {
    angularDampingOverride=1
    angularSpringConstant=1500
    bAngularSpring=true
    bGravityOverrideInSimSpace=true
    bOverrideAngularDamping=true
    gravityScale=0.35
    simulationSpace=RootRelative
    alphaBoolBlend {blendOption=Linear}
    alphaScaleBiasClamp {scale=0.25}
    boundBone {boneName=pouch_2_r}
    componentLinearAccScale {z=0.5}
    componentLinearVelScale {z=0.5}
    physicsBodyDefinitions [{
      sphereCollisionRadius=0
      boundBone {boneName=pouch_2_r}
      boxExtents=<10 20 20>
      constraintSetup {angularLimitsMax=<0 10 0>, angularLimitsMin=<0 -10 -40>, angularTarget=<0 0 0>}
      localJointOffset=<20 0 0>
     }]
   }
   n23 = Anim Dynamics - Bone: pouch_3_r(ComponentPose=@n22, Alpha=1, ExternalForce="0,0,0") {
    angularDampingOverride=1
    angularSpringConstant=1500
    bAngularSpring=true
    bGravityOverrideInSimSpace=true
    bOverrideAngularDamping=true
    gravityScale=0.35
    simulationSpace=RootRelative
    alphaBoolBlend {blendOption=Linear}
    alphaScaleBiasClamp {scale=0.6}
    boundBone {boneName=pouch_3_r}
    componentLinearAccScale {z=0.5}
    componentLinearVelScale {z=0.5}
    physicsBodyDefinitions [{
      sphereCollisionRadius=0
      boundBone {boneName=pouch_3_r}
      boxExtents=<10 20 20>
      constraintSetup {angularLimitsMax=<0 0 0>, angularLimitsMin=<0 0 -40>, angularTarget=<0 0 0>}
      localJointOffset=<10 0 0>
     }]
   }
   n24 = Anim Dynamics - Bone: pouch_4_r(ComponentPose=@n23, Alpha=1, ExternalForce="0,0,0") {
    angularDampingOverride=1
    angularSpringConstant=1500
    bAngularSpring=true
    bGravityOverrideInSimSpace=true
    bOverrideAngularDamping=true
    gravityScale=0.35
    simulationSpace=RootRelative
    alphaBoolBlend {blendOption=Linear}
    boundBone {boneName=pouch_4_r}
    componentLinearAccScale {z=0.5}
    componentLinearVelScale {z=0.5}
    physicsBodyDefinitions [{
      sphereCollisionRadius=0
      boundBone {boneName=pouch_4_r}
      boxExtents=<10 20 20>
      constraintSetup {angularLimitsMax=<0 30 0>, angularLimitsMin=<0 -30 -40>, angularTarget=<0 0 0>}
      localJointOffset=<10 0 0>
     }]
   }
   n25 = Anim Dynamics - Bone: pouch_2_l(ComponentPose=@n24, Alpha=1, ExternalForce="0,0,0") {
    angularDampingOverride=1
    angularSpringConstant=1500
    bAngularSpring=true
    bGravityOverrideInSimSpace=true
    bOverrideAngularDamping=true
    gravityScale=0.35
    simulationSpace=RootRelative
    alphaBoolBlend {blendOption=Linear}
    alphaScaleBiasClamp {scale=0.35}
    boundBone {boneName=pouch_2_l}
    componentLinearAccScale {z=0.5}
    componentLinearVelScale {z=0.5}
    physicsBodyDefinitions [{
      sphereCollisionRadius=0
      boundBone {boneName=pouch_2_l}
      boxExtents=<10 20 20>
      constraintSetup {angularLimitsMax=<0 10 0>, angularLimitsMin=<0 -10 -40>, angularTarget=<0 0 0>}
      localJointOffset=<20 0 0>
     }]
   }
   n26 = Anim Dynamics - Bone: pouch_3_l(ComponentPose=@n25, Alpha=1, ExternalForce="0,0,0") {
    angularDampingOverride=1
    angularSpringConstant=1500
    bAngularSpring=true
    bGravityOverrideInSimSpace=true
    bOverrideAngularDamping=true
    gravityScale=0.35
    simulationSpace=RootRelative
    alphaBoolBlend {blendOption=Linear}
    alphaScaleBiasClamp {scale=0.6}
    boundBone {boneName=pouch_3_l}
    componentLinearAccScale {z=0.5}
    componentLinearVelScale {z=0.5}
    physicsBodyDefinitions [{
      sphereCollisionRadius=0
      boundBone {boneName=pouch_3_l}
      boxExtents=<10 20 20>
      constraintSetup {angularLimitsMax=<0 0 0>, angularLimitsMin=<0 0 -40>, angularTarget=<0 0 0>}
      localJointOffset=<10 0 0>
     }]
   }
   n27 = Anim Dynamics - Bone: pouch_4_l(ComponentPose=@n26, Alpha=1, ExternalForce="0,0,0") {
    angularDampingOverride=1
    angularSpringConstant=1500
    bAngularSpring=true
    bGravityOverrideInSimSpace=true
    bOverrideAngularDamping=true
    gravityScale=0.35
    simulationSpace=RootRelative
    alphaBoolBlend {blendOption=Linear}
    boundBone {boneName=pouch_4_l}
    componentLinearAccScale {z=0.5}
    componentLinearVelScale {z=0.5}
    physicsBodyDefinitions [{
      sphereCollisionRadius=0
      boundBone {boneName=pouch_4_l}
      boxExtents=<10 20 20>
      constraintSetup {angularLimitsMax=<0 30 0>, angularLimitsMin=<0 -30 -40>, angularTarget=<0 0 0>}
      localJointOffset=<10 0 0>
     }]
   }
   n28 = "Constraint - chest_pouch_1_l (PPPPPP)"!(
    ComponentPose=@n27
    Alpha=1
    ConstraintWeights_0=0.1035
    ConstraintWeights_1:float=0.2638
    ConstraintWeights_2:float=0.3081
    ConstraintWeights_3:float=0.2437
    ConstraintWeights_4:float=0.0314
    ConstraintWeights_5:float=0.0495
   ) -> (Pose) {
    alphaBoolBlend {blendOption=Linear}
    boneToModify {boneName=chest_pouch_1_l}
    constraintSetup [
     {targetBone {boneName=spine_01}}
     {targetBone {boneName=spine_02}}
     {targetBone {boneName=spine_03}}
     {targetBone {boneName=spine_04}}
     {targetBone {boneName=spine_05}}
     {targetBone {boneName=pelvis}}
    ]
    constraintWeights [1,1,1,1,1,1]
   }
   n29 = Anim Dynamics - Bone: chest_pouch_2_l(ComponentPose=@n28, Alpha=1, ExternalForce="0,0,0") {
    angularDampingOverride=1
    angularSpringConstant=1500
    bAngularSpring=true
    bGravityOverrideInSimSpace=true
    bOverrideAngularDamping=true
    gravityScale=0.35
    simulationSpace=RootRelative
    alphaBoolBlend {blendOption=Linear}
    alphaScaleBiasClamp {scale=0.6}
    boundBone {boneName=chest_pouch_2_l}
    componentLinearAccScale {z=0.5}
    componentLinearVelScale {z=0.5}
    physicsBodyDefinitions [{
      sphereCollisionRadius=0
      boundBone {boneName=chest_pouch_2_l}
      boxExtents=<10 20 20>
      constraintSetup {angularLimitsMax=<0 0 0>, angularLimitsMin=<0 0 -40>, angularTarget=<0 0 0>}
      localJointOffset=<10 0 0>
     }]
   }
   n30 = Anim Dynamics - Bone: chest_pouch_3_l(ComponentPose=@n29, Alpha=1, ExternalForce="0,0,0") {
    angularDampingOverride=1
    angularSpringConstant=1500
    bAngularSpring=true
    bGravityOverrideInSimSpace=true
    bOverrideAngularDamping=true
    gravityScale=0.35
    simulationSpace=RootRelative
    alphaBoolBlend {blendOption=Linear}
    boundBone {boneName=chest_pouch_3_l}
    componentLinearAccScale {z=0.5}
    componentLinearVelScale {z=0.5}
    physicsBodyDefinitions [{
      sphereCollisionRadius=0
      boundBone {boneName=chest_pouch_3_l}
      boxExtents=<10 20 20>
      constraintSetup {angularLimitsMax=<0 30 0>, angularLimitsMin=<0 -30 -40>, angularTarget=<0 0 0>}
      localJointOffset=<10 0 0>
     }]
   }
   n31 = "Constraint - chest_pouch_1_r (PPPPPPP)"!(
    ComponentPose=@n30
    Alpha=1
    ConstraintWeights_0=0.1001
    ConstraintWeights_1:float=0.2577
    ConstraintWeights_2:float=0.3072
    ConstraintWeights_3:float=0.2555
    ConstraintWeights_4:float=0.033
    ConstraintWeights_5:float=0.0002
    ConstraintWeights_6:float=0.0463
   ) -> (Pose) {
    alphaBoolBlend {blendOption=Linear}
    boneToModify {boneName=chest_pouch_1_r}
    constraintSetup [
     {targetBone {boneName=spine_01}}
     {targetBone {boneName=spine_02}}
     {targetBone {boneName=spine_03}}
     {targetBone {boneName=spine_04}}
     {targetBone {boneName=spine_05}}
     {targetBone {boneName=clavicle_r}}
     {targetBone {boneName=pelvis}}
    ]
    constraintWeights [1,1,1,1,1,1,1]
   }
   n32 = Anim Dynamics - Bone: chest_pouch_2_r(ComponentPose=@n31, Alpha=1, ExternalForce="0,0,0") {
    angularDampingOverride=1
    angularSpringConstant=1500
    bAngularSpring=true
    bGravityOverrideInSimSpace=true
    bOverrideAngularDamping=true
    gravityScale=0.35
    simulationSpace=RootRelative
    alphaBoolBlend {blendOption=Linear}
    alphaScaleBiasClamp {scale=0.6}
    boundBone {boneName=chest_pouch_2_r}
    componentLinearAccScale {z=0.5}
    componentLinearVelScale {z=0.5}
    physicsBodyDefinitions [{
      sphereCollisionRadius=0
      boundBone {boneName=chest_pouch_2_r}
      boxExtents=<10 20 20>
      constraintSetup {angularLimitsMax=<0 0 0>, angularLimitsMin=<0 0 -40>, angularTarget=<0 0 0>}
      localJointOffset=<10 0 0>
     }]
   }
   n33 = Anim Dynamics - Bone: chest_pouch_3_r(ComponentPose=@n32, Alpha=1, ExternalForce="0,0,0") {
    angularDampingOverride=1
    angularSpringConstant=1500
    bAngularSpring=true
    bGravityOverrideInSimSpace=true
    bOverrideAngularDamping=true
    gravityScale=0.35
    simulationSpace=RootRelative
    alphaBoolBlend {blendOption=Linear}
    boundBone {boneName=chest_pouch_3_r}
    componentLinearAccScale {z=0.5}
    componentLinearVelScale {z=0.5}
    physicsBodyDefinitions [{
      sphereCollisionRadius=0
      boundBone {boneName=chest_pouch_3_r}
      boxExtents=<10 20 20>
      constraintSetup {angularLimitsMax=<0 30 0>, angularLimitsMin=<0 -30 -40>, angularTarget=<0 0 0>}
      localJointOffset=<10 0 0>
     }]
   }
   n34 = "Constraint - kneepad_l (PP)"!(ComponentPose=@n33, Alpha=1, ConstraintWeights_0=0.65, ConstraintWeights_1:float=0.35) -> (Pose) {
    alphaBoolBlend {blendOption=Linear}
    boneToModify {boneName=kneepad_l}
    constraintSetup [{targetBone {boneName=thigh_l}}, {targetBone {boneName=calf_l}}]
    constraintWeights [1,1]
   }
   n35 = "Constraint - kneepad_r (PP)"!(ComponentPose=@n34, Alpha=1, ConstraintWeights_0=0.65, ConstraintWeights_1:float=0.35) -> (Pose) {
    alphaBoolBlend {blendOption=Linear}
    boneToModify {boneName=kneepad_r}
    constraintSetup [{targetBone {boneName=thigh_r}}, {targetBone {boneName=calf_r}}]
    constraintWeights [1,1]
   }
   n36 = Get knee_r() -> (knee_r:double)
   n37 = "Transform (Modify) Bone - Bone: gastroc_r"(ComponentPose=@n35, Alpha=@n36, Translation="-8,0,0", Rotation="0,0,0", Scale="1,1,1") {
    rotationMode=BMM_Additive
    rotationSpace=BCS_BoneSpace
    scaleMode=BMM_Additive
    scaleSpace=BCS_BoneSpace
    translationMode=BMM_Additive
    translationSpace=BCS_BoneSpace
    alphaBoolBlend {blendOption=Linear}
    alphaScaleBiasClamp {interpSpeedDecreasing=20, interpSpeedIncreasing=15}
    boneToModify {boneName=gastroc_r}
    translation {x=-8}
   }
   n38 = "Transform (Modify) Bone - Bone: poplit_r"(ComponentPose=@n37, Alpha=@n36, Translation="4,0,0", Rotation="0,0,0", Scale="1,1,1") {
    rotationMode=BMM_Additive
    rotationSpace=BCS_BoneSpace
    scaleMode=BMM_Additive
    scaleSpace=BCS_BoneSpace
    translationMode=BMM_Additive
    translationSpace=BCS_BoneSpace
    alphaBoolBlend {blendOption=Linear}
    alphaScaleBiasClamp {interpSpeedDecreasing=20, interpSpeedIncreasing=15}
    boneToModify {boneName=poplit_r}
    translation {x=4}
   }
   n39 = "Transform (Modify) Bone - Bone: shin_r"(ComponentPose=@n38, Alpha=@n36, Translation="0,0,0", Rotation="0,-30,0", Scale="1,1,1") {
    rotationMode=BMM_Additive
    rotationSpace=BCS_BoneSpace
    scaleMode=BMM_Additive
    scaleSpace=BCS_BoneSpace
    translationMode=BMM_Additive
    translationSpace=BCS_BoneSpace
    alphaBoolBlend {blendOption=Linear}
    alphaScaleBiasClamp {interpSpeedDecreasing=20, interpSpeedIncreasing=15}
    boneToModify {boneName=shin_r}
    rotation {yaw=-30}
   }
   n40 = Get knee_l() -> (knee_l:double)
   n41 = "Transform (Modify) Bone - Bone: gastroc_l"(ComponentPose=@n39, Alpha=@n40, Translation="-6,-0.000,0.000", Rotation="0,0,0", Scale="1,1,1") {
    rotationMode=BMM_Additive
    rotationSpace=BCS_BoneSpace
    scaleMode=BMM_Additive
    scaleSpace=BCS_BoneSpace
    translationMode=BMM_Additive
    translationSpace=BCS_BoneSpace
    alphaBoolBlend {blendOption=Linear}
    alphaScaleBiasClamp {interpSpeedDecreasing=20, interpSpeedIncreasing=15}
    boneToModify {boneName=gastroc_l}
    translation {x=-6}
   }
   n42 = "Transform (Modify) Bone - Bone: poplit_l"(ComponentPose=@n41, Alpha=@n40, Translation="4,0.000,0.000", Rotation="0,0,0", Scale="1,1,1") {
    rotationMode=BMM_Additive
    rotationSpace=BCS_BoneSpace
    scaleMode=BMM_Additive
    scaleSpace=BCS_BoneSpace
    translationMode=BMM_Additive
    translationSpace=BCS_BoneSpace
    alphaBoolBlend {blendOption=Linear}
    alphaScaleBiasClamp {interpSpeedDecreasing=20, interpSpeedIncreasing=15}
    boneToModify {boneName=poplit_l}
    translation {x=4}
   }
   n43 = "Transform (Modify) Bone - Bone: shin_l"(ComponentPose=@n42, Alpha=@n40, Translation="0,0,0", Rotation="0.000,-30,0.000", Scale="1,1,1") {
    rotationMode=BMM_Additive
    rotationSpace=BCS_BoneSpace
    scaleMode=BMM_Additive
    scaleSpace=BCS_BoneSpace
    translationMode=BMM_Additive
    translationSpace=BCS_BoneSpace
    alphaBoolBlend {blendOption=Linear}
    alphaScaleBiasClamp {interpSpeedDecreasing=20, interpSpeedIncreasing=15}
    boneToModify {boneName=shin_l}
    rotation {yaw=-30}
   }
   n44 = "Transform (Modify) Bone - Bone: kneepad_2_r"(ComponentPose=@n43, Alpha=@n36, Translation="-1,2.2,0", Rotation="0,0,0", Scale="0.6,1.000,1.000") {
    scaleMode=BMM_Additive
    scaleSpace=BCS_BoneSpace
    translationMode=BMM_Additive
    translationSpace=BCS_ParentBoneSpace
    alphaBoolBlend {blendOption=Linear}
    alphaScaleBiasClamp {interpSpeedDecreasing=20, interpSpeedIncreasing=15}
    boneToModify {boneName=kneepad_2_r}
    scale {x=0.6}
    translation=<-1 2.2>
   }
   n45 = "Transform (Modify) Bone - Bone: kneepad_2_l"(ComponentPose=@n44, Alpha=@n40, Translation="2,-2.2,0.000", Rotation="0,0,0", Scale="0.8,1.000,1.000") {
    scaleMode=BMM_Additive
    scaleSpace=BCS_BoneSpace
    translationMode=BMM_Additive
    translationSpace=BCS_ParentBoneSpace
    alphaBoolBlend {blendOption=Linear}
    alphaScaleBiasClamp {interpSpeedDecreasing=20, interpSpeedIncreasing=15}
    boneToModify {boneName=kneepad_2_l}
    scale {x=0.8}
    translation=<2 -2.2>
   }
   n46 = Get legUp_L() -> (legUp_L:double)
   n47 = "Transform (Modify) Bone - Bone: hip_l"(ComponentPose=@n45, Alpha=@n46, Translation="5,0,0.000", Rotation="0,0,0", Scale="0.6,1.000,1.000") {
    scaleMode=BMM_Additive
    scaleSpace=BCS_BoneSpace
    translationMode=BMM_Additive
    translationSpace=BCS_ParentBoneSpace
    alphaBoolBlend {blendOption=Linear}
    boneToModify {boneName=hip_l}
    scale {x=0.6}
    translation {x=5}
   }
   n48 = "Transform (Modify) Bone - Bone: thighFront_l"(ComponentPose=@n47, Alpha=@n46, Translation="-3,0,0.000", Rotation="0,0,0", Scale="0.6,1.000,1.000") {
    scaleMode=BMM_Additive
    scaleSpace=BCS_BoneSpace
    translationMode=BMM_Additive
    translationSpace=BCS_ParentBoneSpace
    alphaBoolBlend {blendOption=Linear}
    boneToModify {boneName=thighFront_l}
    scale {x=0.6}
    translation {x=-3}
   }
   n49 = "Transform (Modify) Bone - Bone: thighBack_l"(ComponentPose=@n48, Alpha=@n46, Translation="7,-3,0", Rotation="0,0,0", Scale="0.6,1.000,1.000") {
    scaleMode=BMM_Additive
    scaleSpace=BCS_BoneSpace
    translationMode=BMM_Additive
    translationSpace=BCS_ParentBoneSpace
    alphaBoolBlend {blendOption=Linear}
    boneToModify {boneName=thighBack_l}
    scale {x=0.6}
    translation=<7 -3>
   }
   Output Pose(Result=Component To Local(ComponentPose=@n49)) {layerGroup=None}
  ]}
 EventGraph {nodes [
   Event BlueprintUpdateAnimation()
   n1 = Get Skel_Mesh()
   > Set shldr_raise_r(shldr_raise_r:double=GetAlphaFromReferenceTransformDelta(
     SkeletalMeshComponent=@n1
     BoneName=clavicle_r
     BaseName=spine_05
     DeltaComponentType=QuaternionTwist
     InputRange=(X=-2,Y=-7)
     TwistAxis="0,0,1"
    )) -> (Output_Get:double)
   > Set arm_raise_r(arm_raise_r:double=GetAlphaFromReferenceTransformDelta(
     SkeletalMeshComponent=@n1
     BoneName=upperarm_r
     BaseName=spine_05
     DeltaComponentType=QuaternionTwist
     InputRange=(X=-120,Y=-80)
     TwistAxis="0,0,-1"
    ).RawOutput) -> (Output_Get:double)
   > Set knee_r(knee_r:double=GetAlphaFromReferenceTransformDelta(SkeletalMeshComponent=Get Skel_Mesh(), BoneName=calf_r, BaseName=thigh_r, InputRange=(X=30,Y=120), TwistAxis="0,0,1")) -> (Output_Get:double)
   > Set knee_l(knee_l:double=GetAlphaFromReferenceTransformDelta(SkeletalMeshComponent=Get Skel_Mesh(), BoneName=calf_l, BaseName=thigh_l, InputRange=(X=30,Y=120), TwistAxis="0,0,1")) -> (Output_Get:double)
   > Set legUp_L(legUp_L:double=GetAlphaFromReferenceTransformDelta(
     SkeletalMeshComponent=Get Skel_Mesh()
     BoneName=thigh_l
     BaseName=pelvis
     DeltaComponentType=QuaternionTwist
     InputRange=(X=25,Y=95)
     TwistAxis="0,0,1"
    )) -> (Output_Get:double)
   Event OnBeginCrouching()
   > Set crouch_alpha(crouch_alpha:double=1) -> (Output_Get:double)
   Event OnEndCrouching()
   > Set crouch_alpha(crouch_alpha=0)
   Event BlueprintInitializeAnimation()
   > Set Skel_Mesh(Skel_Mesh=GetOwnerSkeletalMeshComponent())
  ]}
]
variables [
 arm_raise_r {type=number}
 crouch_alpha {type=number}
 knee_l {type=number}
 knee_r {type=number}
 legUp_L {type=number}
 shldr_raise_r {type=number}
 Skel_Mesh {type=SkeletalMeshComponent}
]
